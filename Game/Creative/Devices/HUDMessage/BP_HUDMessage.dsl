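; Creative HUD Message device that shows configurable text to players on a timer from minigame start or when triggered.
; The server writes replicated show time, trigger data and a hide counter, whose OnRep functions drive client-side display and hiding.
; Triggers received outside an active minigame are cached, and timers are cleared at game end.
; Messages carry team, sound, type and DataTable-matched rich text style, and are sent through the device's volume.
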(event EventOnMinigameStarted
  ;[1 1 1] Start Minigame
  (|CallParentFunction)
  (Utilities|IsMessageonATimer?
    (:Yes
      (CallFunction|UpdateRoundStartStartTime)
      (CallFunction|ShowViaTimer))))

(event Custom|OnPlayerAdded
  (params (UniqueNetId :type Unique%20Net%20Id%20Structure) (IsLocalPlayer :type Boolean))
  ;[1 1 1] Start Minigame
  (if IsLocalPlayer
    (Utilities|IsinProgress?
      (:Yes
        (Utilities|FlowControl|Sequence
          (:then_0
            (Utilities|IsMessageonATimer?
              (:Yes
                (CallFunction|ShowViaTimer))))
          (:then_1
            (Default|ShouldShowMessageViaTrigger?
              (:Yes
                (CallFunction|ShowViaTrigger (Variables|Transient|GetCachedTriggerInstigator)))))))
      (:No%20Minigame
        (CallFunction|LogNoMinigameatonAddedtoMinigame)))))

(event Custom|UpdateRoundStartStartTime
  ;[1 1 1] Setters
  (Networking|SwitchHasAuthority
    (:Authority
      (Networking|FlushNetDormancy)
      (Variables|Replication|SetServerTimeWhenMessgeWillShow (+ (Fortnite|Sync|Time|GetServerWorldTimeSeconds) (Math|Conversions|ToFloat(Integer) (Variables|ToyProperties|GetTimeFromRoundStartOption)))))))

(event EventOnAddedToMinigame
  (params (Minigame :type Fort%20Minigame%20Object%20Reference))
  ;[1 1 1] Start Minigame
  (|CallParentFunction Minigame)
  (Minigame|BindEventtoPlayerRemovedDelegate Minigame (EventDispatchers|CreateEvent :!event Custom|OnPlayerRemoved))
  (Minigame|BindEventtoPlayerAddedDelegate Minigame (EventDispatchers|CreateEvent :!event Custom|OnPlayerAdded))
  (Utilities|IsinProgress?
    (:Yes
      (Utilities|FlowControl|Sequence
        (:then_0
          (Utilities|IsMessageonATimer?
            (:Yes
              (CallFunction|ShowViaTimer))))
        (:then_1
          (Default|ShouldShowMessageViaTrigger?
            (:Yes
              (CallFunction|ShowViaTrigger (Variables|Transient|GetCachedTriggerInstigator)))))))
    (:No%20Minigame
      (CallFunction|LogNoMinigameatonAddedtoMinigame))))

(event EventOnRemovedFromMinigame
  (params (Minigame :type Fort%20Minigame%20Object%20Reference))
  ;[1 1 1] End Minigame
  (Minigame|UnbindEventfromPlayerRemovedDelegate Minigame (EventDispatchers|CreateEvent :!event Custom|OnPlayerRemoved))
  (Minigame|UnbindEventfromPlayerAddedDelegate Minigame (EventDispatchers|CreateEvent :!event Custom|OnPlayerAdded)))

(event EventOnMinigameEnded
  ;[1 1 1] End Minigame
  (|CallParentFunction)
  (CallFunction|CancelShowMessageTimer))

(event Custom|Cancel%20Show%20Message%20Timer
  ;[1 1 1] End Minigame
  (Networking|CanExecuteCosmeticEvents
    (:True
      ;;[1 0.5 1] Make sure no messages play after the round / game is over
      (Utilities|Time|ClearandInvalidateTimerbyHandle (Variables|Transient|GetShowMessageTimerHandle)))))

(event OnGameplayMessageReceived(ShowWhenReceived)
  (params (TriggerInstigator :type Controller%20Object%20Reference)))

(event Custom|ShowWhenReceivingFrom
  (params (Instigator :type Controller%20Object%20Reference))
  ;[1 1 1] Triggers
  (CallFunction|ShowViaTrigger Instigator))

(event Custom|UpdateShowReceiverData
  (params (TriggeringPlayerState :type Player%20State%20Object%20Reference))
  ;[1 1 1] Setters
  (Networking|SwitchHasAuthority
    (:Authority
      (Networking|FlushNetDormancy)
      (Variables|Replication|SetShowReceiverData (Utilities|Struct|MakeHUDMessageMessageReceivedData (Utilities|Time|GetTimeSeconds) TriggeringPlayerState)))))

(event Custom|ShowViaTrigger
  (params (TriggerInstigator :type Controller%20Object%20Reference))
  ;[1 1 1] Triggers
  (Utilities|IsinProgress?
    (:Yes
      (CallFunction|UpdateShowReceiverData (Helpers|GetPlayerStatefromController TriggerInstigator))
      (Variables|Transient|SetbShowWhenMinigameAvailable false))
    (:No%20Minigame
      (Variables|Transient|SetCachedTriggerInstigator TriggerInstigator)
      (Variables|Transient|SetbShowWhenMinigameAvailable true))))

(event OnGameplayMessageReceived(HideWhenReceived)
  (params (TriggerInstigator :type Controller%20Object%20Reference)))

(event Custom|OnPlayerRemoved
  (params (UniqueNetId :type Unique%20Net%20Id%20Structure) (bIsLocalPlayer :type Boolean))
  ;[1 1 1] Triggers
  (if bIsLocalPlayer
    (CallFunction|HideMessage true)
    (Utilities|Time|ClearandInvalidateTimerbyHandle (Variables|Transient|GetShowMessageTimerHandle))))

(event Custom|HideWhenRecevingFrom
  (params (Instigator :type Controller%20Object%20Reference))
  ;[1 1 1] Triggers
  (CallFunction|IncrementHideReceiverFlashCount))

(event Custom|IncrementHideReceiverFlashCount
  ;[1 1 1] Setters
  (Networking|SwitchHasAuthority
    (:Authority
      (Networking|FlushNetDormancy)
      (Variables|Replication|SetHideReceiverFlashCount (Math|Byte|Byte+Byte (Variables|Replication|GetHideReceiverFlashCount) 1)))))

(event Custom|ShowViaTimer
  ;[1 1 1] Show Message
  (Utilities|WhentoShowTimedMessage?
    (:Now
      (CallFunction|ShowMessageNow))
    (:Later
      (CallFunction|ShowMessageLater))))

(event Custom|HideMessage
  (params (ClearQueue :type Boolean))
  ;[1 1 1] Hide Message
  (Networking|CanExecuteCosmeticEvents
    (:True
      (bind _asfort_game_state_athena (Utilities|Casting|CastToFortGameStateAthena (Game|GetGameState))
        (:then
          (UI|Messaging|HandleHideGameModeMessageRequest _asfort_game_state_athena "DefaultMessage" false ClearQueue (Class|FortGameplayMessageComponentBase|GetMessageChannelIndex (Variables|Default|GetHideWhenReceived))))
        (:CastFailed
          (CallFunction|LogExpectingGameStatetobeAthena))))))

(event Custom|ShowMessageNow
  ;[1 1 1] Show Message
  (CallFunction|CancelShowMessageTimer)
  (CallFunction|ShowMessage))

(event Custom|LogNoMinigameAtOnAddedToMinigame
  ;[1 1 1] Warning Logging
  (Class|CreativeBlueprintLibrary|LogHUDMessageWarning (Utilities|String|ToString(Text) (Utilities|Text|FormatText "INVTEXT(\"There is no minigame when the {Fort Minigame Progress} component on the {Device} device was added to the minigame. This shouldn\\'t ever happen. Message for reference: {Message}\")" (Utilities|Text|ToText(String) (Utilities|GetObjectName (Variables|Default|GetFortMinigameProgress))) (Utilities|GetDeviceNameasText) (Utilities|GetMessageasText)))))

(event Custom|ShowMessageLater
  ;[1 1 1] Show Message
  (Networking|CanExecuteCosmeticEvents
    (:True
      (bind _returnvalue (Utilities|Time|SetTimerbyEvent (EventDispatchers|CreateEvent :!event Custom|ShowMessage) (Helpers|GetTimeIntervalUntilMessageIsShown)))
      (Variables|Transient|SetShowMessageTimerHandle _returnvalue))))

(event Custom|LogNoVolumeManagerWhenSendingMessage
  ;[1 1 1] Warning Logging
  (Class|CreativeBlueprintLibrary|LogHUDMessageWarning (Utilities|String|ToString(Text) (Utilities|Text|FormatText "INVTEXT(\"There is no volume manager on {Device} when trying to send the message. Message for reference: {Message}\")" (Utilities|GetDeviceNameasText) (Utilities|GetMessageasText)))))

(event Custom|ShowMessage
  (bind (_messagetype _msgflags) (Helpers|GetMessageTypefromOption))
  (bind (_showreceiverdata_flashservertime_10_a847a286461f3a2fa72accacb8b4c9b2 _showreceiverdata_triggerinstigatorplayerstate_9_105646c24495722d237c21b1c1eddcee) (Utilities|Struct|BreakHUDMessageMessageReceivedData (Variables|GetShowReceiverData_FlashServerTime_10_A847A286461F3A2FA72ACCACB8B4C9B2)))
  ;[1 1 1] Show Message
  (Networking|CanExecuteCosmeticEvents
    (:True
      (bind _volume (Utilities|GetValidVolumetoSendMessage)
        (:HasValidVolume
          (HUDMessage|CallOnShowHUDMessage _volume (Utilities|Struct|MakeHUDMessageData (Variables|ToyProperties|GetPlacementOption) _showreceiverdata_triggerinstigatorplayerstate_9_105646c24495722d237c21b1c1eddcee (Helpers|GetMessageTargets) "None" (Class|FortGameplayMessageComponentBase|GetMessageChannelIndex (Variables|Default|GetHideWhenReceived)) _messagetype (Helpers|FormatIntoRichText (Variables|ToyProperties|GetMessageOption)) (Helpers|GetMessageSound) 0 (Helpers|IsMessageTeamBased) (Helpers|TeamSettingtoActorTeam) (Variables|ToyProperties|GetDisplayTimeOption) _msgflags)))))))

(event Custom|LogNoVolumeForThisDeviceWhenSendingMessage
  ;[1 1 1] Warning Logging
  (Class|CreativeBlueprintLibrary|LogHUDMessageWarning (Utilities|String|ToString(Text) (Utilities|Text|FormatText "INVTEXT(\"There is no volume for {Device} when trying to send the message. Message for reference: {Message}\")" (Utilities|GetDeviceNameasText) (Utilities|GetMessageasText)))))

(event Custom|LogExpectingGameStateToBeAthena
  ;[1 1 1] Warning Logging
  (Class|CreativeBlueprintLibrary|LogHUDMessageWarning (Utilities|String|ToString(Text) (Utilities|Text|FormatText "INVTEXT(\"Expecting game state ({Game State}) to be of type FortGameStateAthena. Message for reference: {Message}\")" (Utilities|Text|ToText(String) (Utilities|GetObjectName (Game|GetGameState))) (Utilities|GetMessageasText)))))

(fn FormatIntoRichText
  (params (Message :type String))
  ;[0.5 1 1] Impure to avoid multiple calls
  (bind _textstyle (Helpers|GetTextStyle))
  (if (Utilities|String|EqualExactly(String) _textstyle)
    ;[1 1 0.5] No formatting needed
    (return (Utilities|Text|ToText(String) (Variables|ToyProperties|GetMessageOption)))
    (else
      ;[1 0.5 1] Format into richtext
      (return (Utilities|Text|FormatText "NSLOCTEXT(\"[EE5097AC4ACA792766AD218F41D7B754]\", \"2C193EE842A798BBE067E6BF2A06CB68\", \"<{Style}>{Message}</>\")" (Utilities|Text|ToText(String) _textstyle) (Utilities|Text|ToText(String) (Variables|ToyProperties|GetMessageOption)))))))

(fn GetMessageSound
  (return (Utilities|Select :Index (Variables|ToyProperties|GetSoundOption) :NewEnumerator1 "/Game/Athena/Sounds/LTMs/Score/ScoreLTM_ImportantMessage_Cue.ScoreLTM_ImportantMessage_Cue" :NewEnumerator2 "/Game/Sounds/Creative/Gadgets/Search+Destroy/SearchDestroy_RoundStart_Countdown_Cue.SearchDestroy_RoundStart_Countdown_Cue" :NewEnumerator3 "/Game/Sounds/Creative/Gadgets/Search+Destroy/SearchDestroy_RoundStart_Start_Cue.SearchDestroy_RoundStart_Start_Cue" :NewEnumerator4 "/Game/Sounds/Creative/Gadgets/Search+Destroy/SearchDestroy_Message_Cue.SearchDestroy_Message_Cue")))

(fn GetMessageTargets
  (return (Variables|ToyProperties|GetTeamOption)))

(fn GetMessageTypeFromOption
  (return (Utilities|Select :Index (Variables|ToyProperties|GetMessageTypeOption) :Option%200 "DefaultMessage" :Option%201 "DefaultCriticalMessage" :Option%202 "DefaultCriticalMessage") (Utilities|Select :Index (Variables|ToyProperties|GetMessageTypeOption) :Option%200 "DisableCreativeExclamation" :Option%201 "None" :Option%202 "DisableCreativeExclamation")))

(fn GetPlayerStateFromController
  (params (Controller :type Controller%20Object%20Reference))
  (Utilities|IsValid Controller
    (:Is%20Valid
      (return (Variables|/Script/Engine.Controller|GetPlayerState Controller)))
    (:Is%20Not%20Valid
      (return 0))))

(fn GetTextStyle
  (bind _outrownames (DataTable|GetDataTableRowNames "/Game/Athena/HUD/TextStyles/AthenaScoreAlertStyles.AthenaScoreAlertStyles"))
  (for _array_element _outrownames
    (if (Utilities|String|Equal,CaseInsensitive(String) (Variables|ToyProperties|GetStyleOption) (Utilities|String|ToString(Name) _array_element))
      (return (Utilities|String|ToString(Name) _array_element))))
  (return ""))

(fn GetTimeIntervalUntilMessageIsShown
  (return (- (Variables|Replication|GetServerTimeWhenMessgeWillShow) (Fortnite|Sync|Time|GetServerWorldTimeSeconds))))

(fn IsMessageTeamBased
  (return (Knobs|Team|IsValidTeamIndex (Variables|ToyProperties|GetTeamOption))))

(fn IsVisibleDuringMinigame
  (return (== (Variables|ToyProperties|GetIsVisibleDuringGameOption) 1)))

(fn OnRep_HideReceiverFlashCount
  (Networking|CanExecuteCosmeticEvents
    (:True
      (CallFunction|HideMessage false))))

(fn OnRep_ServerTimeWhenMessgeWillShow
  (CallFunction|ShowViaTimer))

(fn OnRep_ShowReceiverData
  (Default|IsMessageStillTemporarilyRevelantforClients?
    (:Yes
      (CallFunction|ShowMessage))))

(fn ShouldShowTimedMessaged
  (Utilities|FlowControl|Sequence
    (:then_0
      ;[1 1 1] Only show messages on clients
      (Networking|CanExecuteCosmeticEvents
        (:False
          (return false))))
    (:then_1
      ;[1 1 1] There is still time to show the message
      (return (> (Helpers|GetTimeIntervalUntilMessageIsShown) -0.1)))))

(fn TeamSettingToActorTeam
  (return (select (Knobs|Team|IsValidTeamIndex (Variables|ToyProperties|GetTeamOption)) (Math|Conversions|ToInteger(Byte) (Teams|IndextoTeam (Knobs|Team|GetTeamValueasIndex (Variables|ToyProperties|GetTeamOption)))) 0)))

(fn ConstructionScript)
